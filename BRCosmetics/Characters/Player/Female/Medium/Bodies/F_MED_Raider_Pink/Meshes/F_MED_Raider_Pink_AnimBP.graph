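# Animation Blueprint for the female Raider Pink cosmetic body, layering secondary motion and pose fixes on the pose copied from the parent mesh.
# A Control Rig feeds constraints and look-ats for bags, leg pad and zipper, with AnimDynamics swinging the zipper; the arm pad constraint is unconnected.
# Additive bone offsets correct neck, belt, backpack and feet, using alphas the EventGraph derives from LOD, reference-pose deltas, stationary crouch and skydiving with backbling.
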AnimBlueprint /BRCosmetics/Characters/Player/Female/Medium/Bodies/F_MED_Raider_Pink/Meshes/F_MED_Raider_Pink_AnimBP
target_skeleton=&/F_MED_Raider_Pink_Skeleton
graphs [
 AnimGraph {nodes [
   "Constraint - dyn_armPad_r (PP)"!(ComponentPose, Alpha=1, ConstraintWeights_0=0.35, ConstraintWeights_1:float=0.65) -> (Pose) {
    alphaBoolBlend {blendOption=Linear}
    boneToModify {boneName=dyn_armPad_r}
    constraintSetup [{targetBone {boneName=upperarm_twist_02_r}}, {targetBone {boneName=lowerarm_r}}]
    constraintWeights [1,1]
   }
   n1 = Control Rig(Source=Copy Pose From Mesh() {bUseAttachedParent=true}) {
    alphaBoolBlend {blendOption=Linear}
    controlRigAssetReference {blueprintRigClass=&/F_MED_Raider_Pink_ControlRig.F_MED_Raider_Pink_ControlRig_C}
   }
   n2 = "Constraint - dyn_bag_a_MSTR (PPP)"!(
    ComponentPose=Local To Component(LocalPose=@n1)
    Alpha=1
    ConstraintWeights_0=0.036
    ConstraintWeights_1:float=0.238
    ConstraintWeights_2:float=0.726
   ) -> (Pose) {
    alphaBoolBlend {blendOption=Linear}
    boneToModify {boneName=dyn_bag_a_MSTR}
    constraintSetup [{targetBone {boneName=pelvis}}, {targetBone {boneName=thigh_r}}, {targetBone {boneName=thigh_twist_01_r}}]
    constraintWeights [1,1,1]
   }
   n3 = "Constraint - dyn_bag_b_MSTR (PPP)"!(ComponentPose=@n2, Alpha=1, ConstraintWeights_0=0.032, ConstraintWeights_1:float=0.297, ConstraintWeights_2:float=0.671) -> (Pose) {
    alphaBoolBlend {blendOption=Linear}
    boneToModify {boneName=dyn_bag_b_MSTR}
    constraintSetup [{targetBone {boneName=pelvis}}, {targetBone {boneName=thigh_r}}, {targetBone {boneName=thigh_twist_01_r}}]
    constraintWeights [1,1,1]
   }
   n4 = "Constraint - dyn_bag_c_MSTR (PPP)"!(ComponentPose=@n3, Alpha=1, ConstraintWeights_0=0.046, ConstraintWeights_1:float=0.319, ConstraintWeights_2:float=0.635) -> (Pose) {
    alphaBoolBlend {blendOption=Linear}
    boneToModify {boneName=dyn_bag_c_MSTR}
    constraintSetup [{targetBone {boneName=pelvis}}, {targetBone {boneName=thigh_r}}, {targetBone {boneName=thigh_twist_01_r}}]
    constraintWeights [1,1,1]
   }
   n5 = "Constraint - dyn_bag_a_TAT (PP)"!(ComponentPose=@n4, Alpha=1, ConstraintWeights_0=0.647, ConstraintWeights_1:float=0.353) -> (Pose) {
    alphaBoolBlend {blendOption=Linear}
    boneToModify {boneName=dyn_bag_a_TAT}
    constraintSetup [{targetBone {boneName=thigh_r}}, {targetBone {boneName=thigh_twist_01_r}}]
    constraintWeights [1,1]
   }
   n6 = "Constraint - dyn_bag_b_TAT (PP)"!(ComponentPose=@n5, Alpha=1, ConstraintWeights_0=0.67, ConstraintWeights_1:float=0.33) -> (Pose) {
    alphaBoolBlend {blendOption=Linear}
    boneToModify {boneName=dyn_bag_b_TAT}
    constraintSetup [{targetBone {boneName=thigh_r}}, {targetBone {boneName=thigh_twist_01_r}}]
    constraintWeights [1,1]
   }
   n7 = "Constraint - dyn_bag_c_TAT (PP)"!(ComponentPose=@n6, Alpha=1, ConstraintWeights_0=0.67, ConstraintWeights_1:float=0.33) -> (Pose) {
    alphaBoolBlend {blendOption=Linear}
    boneToModify {boneName=dyn_bag_c_TAT}
    constraintSetup [{targetBone {boneName=thigh_r}}, {targetBone {boneName=thigh_twist_01_r}}]
    constraintWeights [1,1]
   }
   n8 = Look At - Bone: dyn_bag_a_AIM(ComponentPose=@n7, Alpha=1) {
    bUseLookUpAxis=true
    alphaBoolBlend {blendOption=Linear}
    boneToModify {boneName=dyn_bag_a_AIM}
    lookAtLocation {x=0}
    lookAtTarget {boneReference {boneName=dyn_bag_a_TAT}}
    lookAt_Axis {axis=<1 0>}
    lookUp_Axis {axis=<0 1>}
   }
   n9 = Look At - Bone: dyn_bag_b_AIM(ComponentPose=@n8, Alpha=1) {
    bUseLookUpAxis=true
    alphaBoolBlend {blendOption=Linear}
    boneToModify {boneName=dyn_bag_b_AIM}
    lookAtLocation {x=0}
    lookAtTarget {boneReference {boneName=dyn_bag_b_TAT}}
    lookAt_Axis {axis=<1 0>}
    lookUp_Axis {axis=<0 1>}
   }
   n10 = Look At - Bone: dyn_bag_c_AIM(ComponentPose=@n9, Alpha=1) {
    bUseLookUpAxis=true
    alphaBoolBlend {blendOption=Linear}
    boneToModify {boneName=dyn_bag_c_AIM}
    lookAtLocation {x=0}
    lookAtTarget {boneReference {boneName=dyn_bag_c_TAT}}
    lookAt_Axis {axis=<1 0>}
    lookUp_Axis {axis=<0 1>}
   }
   n11 = "Constraint - dyn_legPad_l (PP)"!(ComponentPose=@n10, Alpha=1, ConstraintWeights_0=0.1, ConstraintWeights_1:float=0.9) -> (Pose) {
    alphaBoolBlend {blendOption=Linear}
    boneToModify {boneName=dyn_legPad_l}
    constraintSetup [{targetBone {boneName=thigh_twist_01_l}}, {targetBone {boneName=calf_l}}]
    constraintWeights [1,1]
   }
   n12 = "Constraint - dyn_zipper_MSTR (PP)"!(ComponentPose=@n11, Alpha=1, ConstraintWeights_0=0.6, ConstraintWeights_1:float=0.4) -> (Pose) {
    alphaBoolBlend {blendOption=Linear}
    boneToModify {boneName=dyn_zipper_MSTR}
    constraintSetup [{targetBone {boneName=spine_05}}, {targetBone {boneName=clavicle_l}}]
    constraintWeights [1,1]
   }
   n13 = Anim Dynamics - Bone: dyn_zipper(ComponentPose=@n12, Alpha=1) {
    angularDampingOverride=0.95
    bGravityOverrideInSimSpace=true
    bOverrideAngularDamping=true
    bUseGravityOverride=true
    alphaBoolBlend {blendOption=Linear}
    boundBone {boneName=dyn_zipper}
    componentAppliedLinearAccClamp=<3000 3000 3000>
    componentLinearAccScale=<0.3 0.3 0.45>
    componentLinearVelScale=<0.1 0.1 0.1>
    gravityOverride {z=-980}
    physicsBodyDefinitions [{
      sphereCollisionRadius=0
      boundBone {boneName=dyn_zipper}
      boxExtents=<4 8 3>
      constraintSetup {angularLimitsMax=<0 5 0>, angularLimitsMin=<0 -45 -15>, angularTarget=<-0.1883 0 0>}
      localJointOffset=<0.5 0 0>
     }]
   }
   n14 = "Transform (Modify) Bone - Bone: neck_01_bk"(
    ComponentPose=@n13
    Alpha=Get LOD 0 Neck Correction() -> (LOD 0 Neck Correction:float)
    Translation="0,-10,0"
    Rotation="0,0,0"
    Scale="1,1,1"
   ) {
    rotationMode=BMM_Additive
    rotationSpace=BCS_BoneSpace
    translationMode=BMM_Additive
    translationSpace=BCS_BoneSpace
    alphaBoolBlend {blendOption=Linear}
    boneToModify {boneName=neck_01_bk}
    translation {y=-10}
   }
   n15 = Get is lod 1() -> (is lod 1:double)
   n16 = "Transform (Modify) Bone - Bone: neck_01_fr"(ComponentPose=@n14, Alpha=@n15, Translation="0,12,0", Rotation="0,0,0", Scale="1,1,1") {
    rotationMode=BMM_Additive
    rotationSpace=BCS_BoneSpace
    translationMode=BMM_Additive
    translationSpace=BCS_BoneSpace
    alphaBoolBlend {blendOption=Linear}
    boneToModify {boneName=neck_01_fr}
    translation {y=12}
   }
   n17 = "Transform (Modify) Bone - Bone: neck_01_bk"(ComponentPose=@n16, Alpha=@n15, Translation="0,-25,0", Rotation="0,0,0", Scale="1,1,1") {
    rotationMode=BMM_Additive
    rotationSpace=BCS_BoneSpace
    translationMode=BMM_Additive
    translationSpace=BCS_BoneSpace
    alphaBoolBlend {blendOption=Linear}
    boneToModify {boneName=neck_01_bk}
    translation {y=-25}
   }
   n18 = "Transform (Modify) Bone - Bone: neck_01_lt"(ComponentPose=@n17, Alpha=@n15, Translation="0,0,-17", Rotation="0,0,0", Scale="1,1,1") {
    rotationMode=BMM_Additive
    rotationSpace=BCS_BoneSpace
    translationMode=BMM_Additive
    translationSpace=BCS_BoneSpace
    alphaBoolBlend {blendOption=Linear}
    boneToModify {boneName=neck_01_lt}
    translation {z=-17}
   }
   n19 = "Transform (Modify) Bone - Bone: neck_01_rt"(ComponentPose=@n18, Alpha=@n15, Translation="0,0,17", Rotation="0,0,0", Scale="1,1,1") {
    rotationMode=BMM_Additive
    rotationSpace=BCS_BoneSpace
    translationMode=BMM_Additive
    translationSpace=BCS_BoneSpace
    alphaBoolBlend {blendOption=Linear}
    boneToModify {boneName=neck_01_rt}
    translation {z=17}
   }
   n20 = "Transform (Modify) Bone - Bone: belt_L"(
    ComponentPose=@n19
    Alpha=Get is crouching() -> (is crouching:double)
    Translation="-1.650,-0.147,-0.006"
    Rotation="0.004,-10.000,0.005"
    Scale="1,1,1"
   ) {
    rotationMode=BMM_Additive
    rotationSpace=BCS_BoneSpace
    translationMode=BMM_Additive
    translationSpace=BCS_BoneSpace
    alphaBoolBlend {blendOption=Linear}
    alphaScaleBiasClamp {bInterpResult=true, interpSpeedDecreasing=15, interpSpeedIncreasing=15}
    boneToModify {boneName=belt_L}
    rotation=<rot 0.0041 -10 0.005>
    translation=<-1.65 -0.147 -0.006>
   }
   n21 = "Transform (Modify) Bone - Bone: attach_backpack"(
    ComponentPose=@n20
    Alpha=Get backpack_alpha() -> (backpack_alpha:double)
    Translation="-0.001,-2.494,-0.812"
    Rotation="0,0,0"
    Scale="1,1,1"
   ) {
    rotationMode=BMM_Additive
    rotationSpace=BCS_BoneSpace
    translationMode=BMM_Additive
    translationSpace=BCS_BoneSpace
    alphaBoolBlend {blendOption=Linear}
    alphaScaleBiasClamp {interpSpeedDecreasing=15, interpSpeedIncreasing=15}
    boneToModify {boneName=attach_backpack}
    translation=<-0.001 -2.494 -0.812>
   }
   n22 = "Transform (Modify) Bone - Bone: back_foot_l"(
    ComponentPose=@n21
    Alpha=Get foot bending l() -> (foot bending l:float)
    Translation="-2.253,0.010,-0.251"
    Rotation="30.000,0.000,0.000"
    Scale="1,1,1"
   ) {
    rotationMode=BMM_Additive
    rotationSpace=BCS_BoneSpace
    translationMode=BMM_Additive
    translationSpace=BCS_BoneSpace
    alphaBoolBlend {blendOption=Linear}
    alphaScaleBiasClamp {interpSpeedDecreasing=15, interpSpeedIncreasing=15}
    boneToModify {boneName=back_foot_l}
    rotation=<rot 30.0001 0 0>
    translation=<-2.253 0.01 -0.251>
   }
   n23 = "Transform (Modify) Bone - Bone: back_foot_r"(ComponentPose=@n22, Alpha=Get foot bending r() -> (foot bending r:float), Translation="-2.253,0,-0.251", Rotation="30,0,0", Scale="1,1,1") {
    rotationMode=BMM_Additive
    rotationSpace=BCS_BoneSpace
    translationMode=BMM_Additive
    translationSpace=BCS_BoneSpace
    alphaBoolBlend {blendOption=Linear}
    alphaScaleBiasClamp {interpSpeedDecreasing=15, interpSpeedIncreasing=15}
    boneToModify {boneName=back_foot_r}
    rotation {pitch=30}
    translation {x=-2.253, z=-0.251}
   }
   Output Pose(Result=Component To Local(ComponentPose=@n23)) {layerGroup=None}
   Reroute Node(InputPin:Wildcard) -> (OutputPin:Wildcard)
  ]}
 EventGraph {nodes [
   Event BlueprintUpdateAnimation()
   > n1 = Sequence!(execute) -> (then_0, then_1, then_2:Exec)
   n2 = Branch(execute=@n1.then_1, Condition="Equal (Integer)"(A=Get CurrentLOD(), B=0) {is=title K2Node_CallFunction})
   > Set is lod 0(is lod 0:double=1) -> (Output_Get:double)
   Set is lod 0(execute=@n2.else, is lod 0=0)
   n3 = Get SkelMesh()
   Set foot bending l(
    execute=@n1.then_2
    foot bending l:float=GetAlphaFromReferenceTransformDelta(SkeletalMeshComponent=@n3, BoneName=foot_l, BaseName=calf_l, InputRange=(X=0,Y=-40), TwistAxis="1,0,0")
   ) -> (Output_Get:float)
   > Set foot bending r(foot bending r:float=GetAlphaFromReferenceTransformDelta(SkeletalMeshComponent=@n3, BoneName=foot_r, BaseName=calf_r, InputRange=(X=0,Y=-40), TwistAxis="1,0,0")) -> (Output_Get:float)
   n4 = Get CurrentLOD()
   > Set LOD 0 Neck Correction(LOD 0 Neck Correction:float=SelectFloat(
     A=GetAlphaFromReferenceTransformDelta(SkeletalMeshComponent=@n3, BoneName=head, BaseName=spine_05, InputRange=(X=0,Y=-40), TwistAxis="1,0,0")
     B=0
     bPickA="Equal (Integer)"(A=@n4, B=0)
    )) -> (Output_Get:float)
   > n5 = Branch(Condition="Equal (Integer)"(A=Get CurrentLOD(), B=1))
   > Set is lod 1(is lod 1:double=1) -> (Output_Get:double)
   Set is lod 1(execute=@n5.else, is lod 1=0)
   n6 = BreakTransform(InTransform=GetDeltaTransformFromRefPose(self=GetOwnerSkeletalMeshComponent(), BoneName=thigh_r, BaseName=pelvis))
   Set thigh_rot_up_r(execute=@n1.then_0, thigh_rot_up_r:double=MapRangeClamped(Value=BreakRotator(InRot=@n6.Rotation).Yaw, InRangeA=0, InRangeB=-100)) -> (Output_Get:double)
   n7 = BreakTransform(InTransform=GetDeltaTransformFromRefPose(self=GetOwnerSkeletalMeshComponent(), BoneName=thigh_l, BaseName=pelvis))
   > Set thigh_rot_up_l(thigh_rot_up_l:double=MapRangeClamped(Value=BreakRotator(InRot=@n7.Rotation).Yaw, InRangeA=-30, InRangeB=-100)) -> (Output_Get:double)
   n8 = NOT Boolean(A=Get bIsMoving2D())
   n9 = Get bIsCrouching()
   > Set is crouching(is crouching:double=SelectFloat(A=1, B=0, bPickA=AND Boolean(A=@n9, B=@n8))) -> (Output_Get:double)
   > Set isWearingBackpack(isWearingBackpack:bool=CheckForBackbling()) -> (Output_Get:bool)
   n10 = Get bIsSkydiveFloating()
   n11 = Get bIsSkydiveDiveMode()
   > n12 = Branch(Condition=AND Boolean(A=Get isWearingBackpack() -> (isWearingBackpack:bool), B=OR Boolean(A=@n10, B=@n11)))
   > Set backpack_alpha(backpack_alpha:double=1) -> (Output_Get:double)
   Set backpack_alpha(execute=@n12.else, backpack_alpha=0)
   Event BlueprintInitializeAnimation()
   > Set SkelMesh(SkelMesh=GetOwnerSkeletalMeshComponent())
  ]}
]
variables [
 backpack_alpha {type=number}
 foot bending l {type=number}
 foot bending r {type=number}
 is crouching {type=number}
 is lod 0 {type=number}
 is lod 1 {type=number}
 isWearingBackpack {type=boolean}
 LOD 0 Neck Correction {type=number}
 SkelMesh {type=SkeletalMeshComponent}
 thigh_rot_up_l {type=number}
 thigh_rot_up_r {type=number}
]
