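# Control Rig for the face accessory of the pink Raider female character that nudges the dyn_glasses bone as the cheeks rise.
# Each inner cheek bone's height relative to faceAttach is remapped, with clamping, into a weight for its own additive local offset on the glasses.
# The left and right offsets share translation and roll but have opposite pitch, so one cheek tilts the glasses while both together cancel pitch and stack the rest.
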ControlRigBlueprint /BRCosmetics/Characters/Player/Female/Medium/Bodies/F_MED_Raider_Pink/Meshes/Parts/F_MED_Raider_Pink_FaceAcc_CtrlRig
graphs [
 RigVMFunctionLibrary {links [], nodes []}
 RigVMModel {nodes [
   RigUnit_BeginExecution = &RigUnit_BeginExecution()
   /Script/ControlRig.RigUnit_ModifyBoneTransforms(
    BoneToModify=((Bone="dyn_glasses",Transform=(Rotation=(X=0,Y=0,Z=0,W=1),Translation=(X=0,Y=0.2,Z=0.3),Scale3D=(X=1,Y=1,Z=1))))
    Mode=AdditiveLocal
    WeightMaximum=1
    WeightMinimum=0
    Weight=&RigVMFunction_MathFloatRemap(
     SourceMaximum=5.2
     SourceMinimum=3.7
     TargetMinimum=0
     bClamp=true
     Value=&RigUnit_GetRelativeBoneTransform(Bone=L_cheek_inner, Space=faceAttach).Transform.Translation.Z
    ).Result
    BoneToModify.0.Transform.Rotation=/Script/RigVM.RigVMFunction_MathQuaternionFromRotator(Rotator=(Pitch=5,Yaw=0,Roll=-3)) {is=struct RigVMUnitNode}.Result
    ExecutePin=@RigUnit_BeginExecution.ExecutePin
   ) {is=struct RigVMUnitNode}
   > &RigUnit_ModifyBoneTransforms(
    BoneToModify=((Bone="dyn_glasses",Transform=(Rotation=(X=0,Y=0,Z=0,W=1),Translation=(X=0,Y=0.2,Z=0.3),Scale3D=(X=1,Y=1,Z=1))))
    Mode=AdditiveLocal
    WeightMaximum=1
    WeightMinimum=0
    Weight=&RigVMFunction_MathFloatRemap(
     SourceMaximum=5.2
     SourceMinimum=3.7
     TargetMinimum=0
     bClamp=true
     Value=&RigUnit_GetRelativeBoneTransform(Bone=R_cheek_inner, Space=faceAttach).Transform.Translation.Z
    ).Result
    BoneToModify.0.Transform.Rotation=&RigVMFunction_MathQuaternionFromRotator(Rotator=(Pitch=-5,Yaw=0,Roll=-3)).Result
   )
  ]}
]
hierarchy [
 root {loc=<0 0 0>, rot=<0 0 0 1>}
 pelvis {parent=root, loc=<0 -0.5301 98.0571>, rot=<0.0127 0.707 -0.0127 -0.707>}
 spine_01 {parent=pelvis, loc=<6.1208 -0.9328 0>, rot=<0 0 0.034 -0.9994>}
 spine_02 {parent=spine_01, loc=<6.5996 -0.2095 0>, rot=<0 0 0.0238 -0.9997>}
 spine_03 {parent=spine_02, loc=<6.4914 -0.1237 0>, rot=<0 0 0.0149 -0.9999>}
 spine_04 {parent=spine_03, loc=<6.9598 0.0794 0>, rot=<0 0 -0.0132 -0.9999>}
 spine_05 {parent=spine_04, loc=<6.9987 0 0>, rot=<0 0 -0.0015 -1>}
 clavicle_l {parent=spine_05, loc=<5.4831 -0.1053 -3.2946>, rot=<0.0465 -0.7289 -0.009 -0.6829>}
 uppelalm_l {parent=clavicle_l, loc=<9.8657 -1.2136 -0.4101>, rot=<0.0164 -0.3245 -0.0269 -0.9454>}
 clavicle_r {parent=spine_05, loc=<5.4831 -0.1053 3.2946>, rot=<0.7289 0.0465 -0.6829 0.009>}
 upperarm_r {parent=clavicle_r, loc=<-9.8658 1.2136 0.4104>, rot=<0.0164 -0.3245 -0.0269 -0.9454>}
 neck_01 {parent=spine_05, loc=<11.3138 0.1529 0>, rot=<0 0 -0.2003 -0.9797>}
 neck_02 {parent=neck_01, loc=<4.5627 0 0>, rot=<0 0 0.0504 -0.9987>}
 head {parent=neck_02, loc=<5.3914 0 0>, rot=<0 0 0.115 -0.9934>}
 dyn_hair_ft_01 {parent=head, loc=<1.9837 5.0678 -7.1296>, rot=<-0.0026 -0.9999 0.0001 0.0101>}
 dyn_hair_ft_02 {parent=dyn_hair_ft_01, loc=<4.2196 0 0>, rot=<-0.0112 -0.157 -0.0571 -0.9859>}
 dyn_hair_ft_03 {parent=dyn_hair_ft_02, loc=<3.2561 0 0>, rot=<-0.0531 -0.2377 -0.0877 -0.9659>}
 dyn_hair_l_01 {parent=head, loc=<2.9954 0.3355 -7.3619>, rot=<-0.1209 -0.989 -0.0111 -0.0847>}
 dyn_hair_l_02 {parent=dyn_hair_l_01, loc=<5.3122 0.0039 -0.0925>, rot=<-0.0135 -0.1098 0.0677 -0.9916>}
 dyn_hair_l_03 {parent=dyn_hair_l_02, loc=<4.2988 -0.0081 0.109>, rot=<-0.0288 -0.1394 0.3028 -0.9424>}
 dyn_hair_l_04 {parent=dyn_hair_l_03, loc=<3.1532 0 -0.1456>, rot=<-0.0464 -0.0817 0.3683 -0.9249>}
 dyn_hair_l_tips {parent=head, loc=<3.0634 3.9555 -7.5866>, rot=<-0.139 -0.8078 -0.1018 -0.5638>}
 faceAttach {parent=head, loc=<-3.6353 -0.7132 -0.0394>, rot=<0.0082 -0.7071 -0.0082 -0.7071>, scale=<0.89 0.89 0.89>}
 C_brow_mid {parent=faceAttach, loc=<-0.0749 10.4916 8.3391>, rot=<0 0.066 -0.013 -0.9977>}
 L_brow_mid {parent=faceAttach, loc=<3.4828 9.7055 9.6359>, rot=<-0.0003 -0.0057 0.2553 -0.9668>}
 L_brow_outer {parent=faceAttach, loc=<5.855 7.5862 10.1794>, rot=<0.0843 0.1022 0.4189 -0.8983>}
 L_cheek_inner {parent=faceAttach, loc=<4.5697 8.6725 3.5697>, rot=<-0.0299 -0.1666 0.3307 -0.9284>}
 C_lip_upper_mid {parent=faceAttach, loc=<-0.0467 10.5258 1.2283>, rot=<0.2225 0.001 -0.0004 -0.9749>}
 L_lip_upper_outer {parent=faceAttach, loc=<1.6678 9.4255 1.1613>, rot=<0.0349 -0.0757 0.2638 -0.961>}
 L_lip_corner {parent=faceAttach, loc=<2.5927 8.9354 1.0104>, rot=<0.0029 0.0081 -0.0146 -0.9999>}
 C_nose_bridge {parent=faceAttach, loc=<-0.0666 11.8914 5.0756>, rot=<-0.2218 -0.0079 -0.0022 -0.9751>}
 R_brow_mid {parent=faceAttach, loc=<-3.5729 9.7055 9.6429>, rot=<-0.0001 0.0044 -0.2553 -0.9669>}
 R_brow_outer {parent=faceAttach, loc=<-5.9565 7.5607 10.172>, rot=<0.0695 -0.0836 -0.4237 -0.8993>}
 R_cheek_inner {parent=faceAttach, loc=<-4.6473 8.6806 3.5601>, rot=<-0.0292 0.1648 -0.3309 -0.9287>}
 R_lip_corner {parent=faceAttach, loc=<-2.6849 8.8845 1.0343>, rot=<0.0029 -0.008 0.0146 -0.9999>}
 C_jaw {parent=faceAttach, loc=<-0.1015 3.0731 1.1502>, rot=<0 0 0 1>}
 L_lip_lower_outer {parent=C_jaw, loc=<1.7558 6.3721 -0.0162>, rot=<0.0132 0.0459 0.3067 -0.9506>}
 C_lip_lower_mid {parent=C_jaw, loc=<0.0586 7.3304 -0.235>, rot=<0 -0.0283 -0.006 -0.9996>}
 R_lip_lower_outer {parent=C_jaw, loc=<-1.7608 6.5029 -0.1495>, rot=<0.0049 -0.0006 -0.272 -0.9623>}
 teeth_lower {parent=C_jaw, loc=<-0.0498 4.6852 -1.9962>, rot=<0 0 0 1>}
 tongue {parent=C_jaw, loc=<0.0583 1.5812 -0.8418>, rot=<-0.1568 0 0 -0.9876>}
 R_eye {parent=faceAttach, loc=<-3.3054 7.2614 8.0517>, rot=<0 0 0.9998 -0.0187>}
 R_eye_lid_upper_mid {parent=R_eye, loc=<0.0372 -1.2959 0.4654>, rot=<0.0523 -0.0009 0.0167 -0.9985>}
 R_eye_lid_lower_mid {parent=R_eye, loc=<-0.098 -0.6161 -1.1793>, rot=<0.0329 -0.0005 0.0167 -0.9993>}
 L_eye {parent=faceAttach, loc=<3.2169 7.2614 8.0517>, rot=<0 0 0.9998 0.0187>}
 L_eye_lid_upper_mid {parent=L_eye, loc=<-0.1297 -1.3423 0.4366>, rot=<0.0523 0 0 -0.9986>}
 L_eye_lid_lower_mid {parent=L_eye, loc=<0.0804 -0.6722 -1.1668>, rot=<0.0329 0 0 -0.9995>}
 teeth_upper {parent=faceAttach, loc=<-0.0492 8.3788 2.0558>, rot=<0 0 0 1>}
 R_lip_upper_outer {parent=faceAttach, loc=<-1.7905 9.4733 1.1892>, rot=<0.0336 0.0793 -0.2651 -0.9604>}
 dyn_glasses {parent=head, loc=<2.7868 7.3366 0>, rot=<0.0029 -0.7071 -0.0029 -0.7071>}
 dyn_hair_bk_l_MSTR {parent=head, loc=<2.5815 -3.563 -6.6016>, rot=<0.0026 -0.9328 0.3452 -0.1036>}
 dyn_hair_bk_l_01 {parent=dyn_hair_bk_l_MSTR, loc=<0 0 0>, rot=<0 0 0 -1>}
 dyn_hair_bk_l_02 {parent=dyn_hair_bk_l_01, loc=<6.1566 0 0>, rot=<-0.0869 0.0301 0.1751 -0.9802>}
 dyn_hair_bk_l_03 {parent=dyn_hair_bk_l_02, loc=<5.6603 0 0>, rot=<-0.0574 0.1096 0.1734 -0.9771>}
 dyn_hair_bk_l_04 {parent=dyn_hair_bk_l_03, loc=<5.7888 0 0>, rot=<-0.1545 -0.2595 0.0419 -0.9524>}
 dyn_hair_bk_r_MSTR {parent=head, loc=<4.2859 -5.7166 5.1554>, rot=<0.047 -0.6269 0.7735 -0.0813>}
 dyn_hair_bk_r_01 {parent=dyn_hair_bk_r_MSTR, loc=<0 0 0>, rot=<0 0 0 1>}
 dyn_hair_bk_r_02 {parent=dyn_hair_bk_r_01, loc=<5.4442 0.1266 -0.0271>, rot=<0 0.0174 0.0274 -0.9995>}
 dyn_hair_bk_r_03 {parent=dyn_hair_bk_r_02, loc=<5.4236 0.0419 -0.0091>, rot=<0.0003 -0.0025 0.0181 -0.9998>}
 dyn_hair_bk_r_04 {parent=dyn_hair_bk_r_03, loc=<4.5294 -0.1688 0.0362>, rot=<0.0026 0.0065 -0.1415 -0.9899>}
 dyn_hair_bk_MSTR {parent=head, loc=<4.9233 -6.5253 0>, rot=<0.0788 -0.7027 0.7027 -0.0788>}
 dyn_hair_bk_01 {parent=dyn_hair_bk_MSTR, loc=<0 0 0>, rot=<0 0 0 -1>}
 dyn_hair_bk_02 {parent=dyn_hair_bk_01, loc=<6 0 0>, rot=<0 0.0143 0 -0.9999>}
 dyn_hair_bk_03 {parent=dyn_hair_bk_02, loc=<5 0 0>, rot=<0 -0.0146 0 -0.9999>}
 dyn_hair_bk_04 {parent=dyn_hair_bk_03, loc=<5.5 0 0>, rot=<0 -0.062 0 -0.9981>}
 dyn_hair_bk_05 {parent=dyn_hair_bk_04, loc=<5 0 0>, rot=<0 -0.0244 0 -0.9997>}
 dyn_hair_r_MSTR {parent=head, loc=<2.0537 -1.1585 6.8265>, rot=<0.019 0.0003 0.9998 0.0057>}
 dyn_hair_r_01 {parent=dyn_hair_r_MSTR, loc=<0 0 0>, rot=<0 0 0 1>}
 dyn_hair_r_02 {parent=dyn_hair_r_01, loc=<4.6011 0 0>, rot=<0.0053 -0.0454 0.0974 -0.9942>}
 dyn_hair_r_03 {parent=dyn_hair_r_02, loc=<4.2088 0 0>, rot=<0.0457 -0.1032 0.1993 -0.9734>}
 dyn_hair_r_04 {parent=dyn_hair_r_03, loc=<4.0628 0 0>, rot=<0.0095 -0.0132 0.0038 -0.9999>}
 dyn_hair_r_05 {parent=dyn_hair_r_04, loc=<4.1325 0 0>, rot=<-0.0854 0.1302 -0.0457 -0.9867>}
 dyn_hair_r_06 {parent=dyn_hair_r_05, loc=<2.7487 0 0>, rot=<0 0.2727 0 -0.9621>}
 neck_01_fr {parent=neck_01, loc=<0 0 0>, rot=<0 0 0 1>}
 neck_01_bk {parent=neck_01, loc=<0 0 0>, rot=<0 0 0 1>}
 neck_01_lt {parent=neck_01, loc=<0 0 0>, rot=<0 0 0 1>}
 neck_01_rt {parent=neck_01, loc=<0 0 0>, rot=<0 0 0 1>}
]
