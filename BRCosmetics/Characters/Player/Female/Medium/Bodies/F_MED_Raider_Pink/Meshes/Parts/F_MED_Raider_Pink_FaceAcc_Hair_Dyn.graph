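# Physics asset for the dynamic hair on the female medium Raider Pink character's FaceAcc hair part.
# Capsule chains cover front, left, right and three back strands with tight swing and twist limits, looser at the right tip, rooted on the head directly or via small MSTR bodies.
# Cross-constraints tie both outer back chains to the center one.
# Head, neck, upper spine, clavicle and right upper arm capsules form the body, joined by free-rotation constraints.
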PhysicsAsset /BRCosmetics/Characters/Player/Female/Medium/Bodies/F_MED_Raider_Pink/Meshes/Parts/F_MED_Raider_Pink_FaceAcc_Hair_Dyn
bodies {
 clavicle_l [{shapeType=Capsule, length=6, radius=6.1299, center=<5.9505 -0.5107 -1.3365>, rotation=<rot 82.9364 -0.0001 -0.0001>}]
 clavicle_r [{shapeType=Capsule, length=5.5, radius=6.6299, center=<-6.6886 0.3471 1.0998>, rotation=<rot 90 178.7341 178.7342>}]
 dyn_hair_bk_01 [{shapeType=Capsule, length=6.5, radius=2, center=<2.8455 0 0>, rotation=<rot 0 0 90>}]
 dyn_hair_bk_02 [{shapeType=Capsule, length=7.25, radius=2, center=<2.2304 0 0>, rotation=<rot 0 0 90>}]
 dyn_hair_bk_03 [{shapeType=Capsule, length=7.75, radius=2, center=<2.2304 0 0>, rotation=<rot 0 0 90>}]
 dyn_hair_bk_04 [{shapeType=Capsule, length=7.5, radius=1, center=<2.2 0 0.42>, rotation=<rot 0 0 90>}]
 dyn_hair_bk_05 [{shapeType=Capsule, length=3.5, radius=2, center=<2.23 0 0.2685>, rotation=<rot 0 0 90>}]
 dyn_hair_bk_MSTR [{shapeType=Capsule, length=0.505, radius=0.505, rotation=<rot 0 -90 0>}]
 dyn_hair_bk_l_01 [{shapeType=Capsule, length=2, radius=2, center=<3 0 0>, rotation=<rot 90 0 0>}]
 dyn_hair_bk_l_02 [{shapeType=Capsule, length=2.5, radius=2.5, center=<3 0 0>, rotation=<rot 90 0 0>}]
 dyn_hair_bk_l_03 [{shapeType=Capsule, length=2.375, radius=1.625, center=<3.2995 -0.0001 0.4268>, rotation=<rot 90 0 0>}]
 dyn_hair_bk_l_04 [{shapeType=Capsule, length=2.5, radius=1.5, center=<2 0 0.54>, rotation=<rot 82 -180 -180>}]
 dyn_hair_bk_l_MSTR [{shapeType=Capsule, length=0.505, radius=0.505, rotation=<rot 0 -90 0>}]
 dyn_hair_bk_r_01 [{shapeType=Capsule, length=2.125, radius=1.875, center=<3 0 0>, rotation=<rot 90 0 0>}]
 dyn_hair_bk_r_02 [{shapeType=Capsule, length=2.5, radius=2, center=<3 0 0>, rotation=<rot 0 0 90>}]
 dyn_hair_bk_r_03 [{shapeType=Capsule, length=2.5, radius=2, center=<3 0 0>, rotation=<rot 0 0 90>}]
 dyn_hair_bk_r_04 [{shapeType=Capsule, length=4.5, radius=2, center=<3.593 -0.5934 0.95>, rotation=<rot 90 -90 -110>}]
 dyn_hair_bk_r_MSTR [{shapeType=Capsule, length=0.505, radius=0.505, rotation=<rot 0 -90 0>}]
 dyn_hair_ft_01 [{shapeType=Capsule, length=2, radius=1.25, center=<1 0 0>, rotation=<rot 90 0 0>}]
 dyn_hair_ft_02 [{shapeType=Capsule, length=2, radius=1.25, center=<1 0 0>, rotation=<rot 90 0 0>}]
 dyn_hair_l_01 [{shapeType=Capsule, length=1.25, radius=1.75, center=<3 0 0>, rotation=<rot 90 0 0>}]
 dyn_hair_l_02 [{shapeType=Capsule, length=1, radius=2.25, center=<2 0 0>, rotation=<rot 90 0 0>}]
 dyn_hair_l_03 [{shapeType=Capsule, length=2, radius=1.5, center=<3.382 -1.1526 0>, rotation=<rot 90 -153.4349 -118.4312>}]
 dyn_hair_r_01 [{shapeType=Capsule, length=2, radius=1.875, center=<2 0 0.4124>, rotation=<rot 86.7553 -180 -180>}]
 dyn_hair_r_02 [{shapeType=Capsule, length=1.875, radius=2, center=<1.9734 -0.0001 0.5354>, rotation=<rot 87.1419 180 176.0556>}]
 dyn_hair_r_03 [{shapeType=Capsule, length=3.925, radius=2.1, center=<2.6772 0 0.8829>, rotation=<rot 85.4579 148.504 163.4008>}]
 dyn_hair_r_04 [{shapeType=Capsule, length=4.25, radius=1.675, center=<5.5207 -0.5072 0.3271>, rotation=<rot 10 0 0>}]
 dyn_hair_r_05 [{shapeType=Capsule, length=3.5, radius=1.45, center=<6.217 -0.5072 0.2256>}]
 dyn_hair_r_MSTR [{shapeType=Capsule, length=0.505, radius=0.505, shapeName=dyn_hair_r_MSTR_capsule}]
 head [{shapeType=Capsule, length=3.2663, radius=7.3967, center=<1.6388 2.7247 1.5555>, rotation=<rot 90 0 -16.637>}]
 neck_01 [{shapeType=Capsule, length=2.3788, radius=4.8, center=<2.6511 -0.1291 -0.0001>, rotation=<rot 90 -11.7683 -5.6279>}]
 spine_04 [
  {shapeType=Capsule, length=9.1886, radius=8.2, center=<6.5959 0 0.0002>}
  {shapeType=Capsule, length=7.5, radius=9.5, shapeName=spine_04_capsule, center=<0.437 2.8132 0>}
 ]
 spine_05 [{shapeType=Capsule, length=10.3457, radius=7.0017, center=<3.6377 -0.2994 0.0001>}]
 upperarm_r [{shapeType=Capsule, length=12.5057, radius=3.1252, center=<-8.0637 -0.0002 0.2758>, rotation=<rot 90 0 0>}]
}
constraints [
 {bone1Name=dyn_hair_ft_01, bone2Name=head, swing1LimitDegrees=20, swing2LimitDegrees=25, twistLimitDegrees=5}
 {bone1Name=dyn_hair_ft_02, bone2Name=dyn_hair_ft_01, swing1LimitDegrees=20, swing2LimitDegrees=25, twistLimitDegrees=5}
 {bone1Name=dyn_hair_l_01, bone2Name=head, swing1LimitDegrees=5, swing2LimitDegrees=15, twistLimitDegrees=2}
 {bone1Name=dyn_hair_l_02, bone2Name=dyn_hair_l_01, swing1LimitDegrees=15, swing2LimitDegrees=15, twistLimitDegrees=2}
 {bone1Name=dyn_hair_l_03, bone2Name=dyn_hair_l_02, swing1LimitDegrees=15, swing2LimitDegrees=15, twistLimitDegrees=2}
 {bone1Name=dyn_hair_bk_01, bone2Name=dyn_hair_bk_MSTR, swing1LimitDegrees=10, swing2LimitDegrees=25, twistLimitDegrees=2}
 {bone1Name=dyn_hair_bk_02, bone2Name=dyn_hair_bk_01, swing1LimitDegrees=10, swing2LimitDegrees=25, twistLimitDegrees=2}
 {bone1Name=dyn_hair_bk_03, bone2Name=dyn_hair_bk_02, swing1LimitDegrees=10, swing2LimitDegrees=25, twistLimitDegrees=2}
 {bone1Name=dyn_hair_bk_04, bone2Name=dyn_hair_bk_03, swing1LimitDegrees=10, swing2LimitDegrees=25, twistLimitDegrees=2}
 {
  bone1Name=head
  bone2Name=neck_01
  swing1LimitDegrees=45
  swing1Motion=Free
  swing2LimitDegrees=45
  swing2Motion=Free
  twistLimitDegrees=45
  twistMotion=Free
 }
 {
  bone1Name=clavicle_l
  bone2Name=spine_05
  swing1LimitDegrees=45
  swing1Motion=Free
  swing2LimitDegrees=45
  swing2Motion=Free
  twistLimitDegrees=45
  twistMotion=Free
 }
 {
  bone1Name=clavicle_r
  bone2Name=spine_05
  swing1LimitDegrees=45
  swing1Motion=Free
  swing2LimitDegrees=45
  swing2Motion=Free
  twistLimitDegrees=45
  twistMotion=Free
 }
 {
  bone1Name=upperarm_r
  bone2Name=clavicle_r
  swing1LimitDegrees=45
  swing1Motion=Free
  swing2LimitDegrees=45
  swing2Motion=Free
  twistLimitDegrees=45
  twistMotion=Free
 }
 {
  bone1Name=neck_01
  bone2Name=spine_05
  swing1LimitDegrees=45
  swing1Motion=Free
  swing2LimitDegrees=45
  swing2Motion=Free
  twistLimitDegrees=45
  twistMotion=Free
 }
 {
  bone1Name=spine_05
  bone2Name=spine_04
  swing1LimitDegrees=45
  swing1Motion=Free
  swing2LimitDegrees=45
  swing2Motion=Free
  twistLimitDegrees=45
  twistMotion=Free
 }
 {bone1Name=dyn_hair_bk_l_01, bone2Name=dyn_hair_bk_l_MSTR, swing1LimitDegrees=10, swing2LimitDegrees=25, twistLimitDegrees=2}
 {bone1Name=dyn_hair_bk_l_02, bone2Name=dyn_hair_bk_l_01, swing1LimitDegrees=10, swing2LimitDegrees=25, twistLimitDegrees=2}
 {bone1Name=dyn_hair_bk_l_03, bone2Name=dyn_hair_bk_l_02, swing1LimitDegrees=10, swing2LimitDegrees=25, twistLimitDegrees=2}
 {bone1Name=dyn_hair_bk_l_04, bone2Name=dyn_hair_bk_l_03, swing1LimitDegrees=10, swing2LimitDegrees=25, twistLimitDegrees=2}
 {bone1Name=dyn_hair_bk_r_01, bone2Name=dyn_hair_bk_r_MSTR, swing1LimitDegrees=5, swing2LimitDegrees=25, twistLimitDegrees=2}
 {bone1Name=dyn_hair_bk_r_02, bone2Name=dyn_hair_bk_r_01, swing1LimitDegrees=5, swing2LimitDegrees=25, twistLimitDegrees=2}
 {bone1Name=dyn_hair_bk_r_03, bone2Name=dyn_hair_bk_r_02, swing1LimitDegrees=10, swing2LimitDegrees=25, twistLimitDegrees=2}
 {bone1Name=dyn_hair_bk_r_04, bone2Name=dyn_hair_bk_r_03, swing1LimitDegrees=10, swing2LimitDegrees=25, twistLimitDegrees=2}
 {
  bone1Name=dyn_hair_bk_l_MSTR
  bone2Name=head
  swing1LimitDegrees=45
  swing1Motion=Free
  swing2LimitDegrees=45
  swing2Motion=Free
  twistLimitDegrees=45
  twistMotion=Free
 }
 {
  bone1Name=dyn_hair_bk_r_MSTR
  bone2Name=head
  swing1LimitDegrees=45
  swing1Motion=Free
  swing2LimitDegrees=45
  swing2Motion=Free
  twistLimitDegrees=45
  twistMotion=Free
 }
 {
  bone1Name=dyn_hair_bk_MSTR
  bone2Name=head
  swing1LimitDegrees=45
  swing1Motion=Free
  swing2LimitDegrees=45
  swing2Motion=Free
  twistLimitDegrees=45
  twistMotion=Free
 }
 {bone1Name=dyn_hair_bk_l_03, bone2Name=dyn_hair_bk_03, swing1LimitDegrees=10, swing2LimitDegrees=25, twistLimitDegrees=2}
 {bone1Name=dyn_hair_r_02, bone2Name=dyn_hair_r_01, swing1LimitDegrees=5, swing2LimitDegrees=15, twistLimitDegrees=2}
 {bone1Name=dyn_hair_r_03, bone2Name=dyn_hair_r_02, swing1LimitDegrees=10, swing2LimitDegrees=35, twistLimitDegrees=2}
 {bone1Name=dyn_hair_r_04, bone2Name=dyn_hair_r_03, swing1LimitDegrees=45, swing2LimitDegrees=45, twistLimitDegrees=2}
 {bone1Name=dyn_hair_r_05, bone2Name=dyn_hair_r_04, swing1LimitDegrees=45, swing2LimitDegrees=45, twistLimitDegrees=2}
 {bone1Name=dyn_hair_bk_05, bone2Name=dyn_hair_bk_04, swing1LimitDegrees=10, swing2LimitDegrees=25, twistLimitDegrees=2}
 {bone1Name=dyn_hair_r_01, bone2Name=dyn_hair_r_MSTR, swing1LimitDegrees=5, swing2LimitDegrees=15, twistLimitDegrees=2}
 {
  bone1Name=dyn_hair_r_MSTR
  bone2Name=head
  swing1LimitDegrees=45
  swing1Motion=Free
  swing2LimitDegrees=45
  swing2Motion=Free
  twistLimitDegrees=45
  twistMotion=Free
 }
 {bone1Name=dyn_hair_bk_r_04, bone2Name=dyn_hair_bk_04, swing1LimitDegrees=5, swing2LimitDegrees=5, twistLimitDegrees=5}
]
