AnimBlueprint /BRCosmetics/Characters/Player/Female/Medium/Bodies/F_MED_Raider_Pink/Meshes/Parts/F_MED_Raider_Pink_FaceAcc_AnimBP
target_skeleton=&/F_MED_Raider_Pink_FaceAcc_Skeleton
graphs [
 AnimGraph {nodes [
   n1 = Get Is DBNO Float() -> (Is DBNO Float:double)
   n2 = "Transform (Modify) Bone - Bone: dyn_hair_bk_l_MSTR"(
    ComponentPose=Local To Component(LocalPose=Use cached pose 'preHairPose'())
    Alpha=@n1
    Translation="0,0,2"
    Rotation="40,0,0"
    Scale="1,1,1"
   ) {
    rotationMode=BMM_Additive
    rotationSpace=BCS_BoneSpace
    translationMode=BMM_Additive
    translationSpace=BCS_BoneSpace
    alphaBoolBlend {blendOption=Linear}
    boneToModify {boneName=dyn_hair_bk_l_MSTR}
    rotation {pitch=40}
    translation {z=2}
   }
   n3 = "Transform (Modify) Bone - Bone: dyn_hair_bk_r_MSTR"(ComponentPose=@n2, Alpha=@n1, Translation="0,0,2", Rotation="40,0,0", Scale="1,1,1") {
    rotationMode=BMM_Additive
    rotationSpace=BCS_BoneSpace
    translationMode=BMM_Additive
    translationSpace=BCS_BoneSpace
    alphaBoolBlend {blendOption=Linear}
    boneToModify {boneName=dyn_hair_bk_r_MSTR}
    rotation {pitch=40}
    translation {z=2}
   }
   n4 = "Transform (Modify) Bone - Bone: dyn_hair_bk_MSTR"(ComponentPose=@n3, Alpha=@n1, Translation="0,0,2", Rotation="40,0,0", Scale="1,1,1") {
    rotationMode=BMM_Additive
    rotationSpace=BCS_BoneSpace
    translationMode=BMM_Additive
    translationSpace=BCS_BoneSpace
    alphaBoolBlend {blendOption=Linear}
    boneToModify {boneName=dyn_hair_bk_MSTR}
    rotation {pitch=40}
    translation {z=2}
   }
   n5 = "Transform (Modify) Bone - Bone: dyn_hair_bk_r_MSTR"(
    ComponentPose=@n4
    Alpha="Get Head Pitch When Crouching "() -> ("Head Pitch When Crouching ":double)
    Translation="0,0,0"
    Rotation="70,0,0"
    Scale="1,1,1"
   ) {
    rotationMode=BMM_Additive
    rotationSpace=BCS_BoneSpace
    translationSpace=BCS_BoneSpace
    alphaBoolBlend {blendOption=Linear}
    alphaScaleBiasClamp {bInterpResult=true, interpSpeedDecreasing=20}
    boneToModify {boneName=dyn_hair_bk_r_MSTR}
    rotation {pitch=70}
   }
   n6 = "Transform (Modify) Bone - Bone: dyn_hair_bk_r_MSTR"(
    ComponentPose=@n5
    Alpha="Get Head Roll When Crouching "() -> ("Head Roll When Crouching ":float)
    Translation="0,0,0"
    Rotation="30,0,0"
    Scale="1,1,1"
   ) {
    rotationMode=BMM_Additive
    rotationSpace=BCS_BoneSpace
    translationSpace=BCS_BoneSpace
    alphaBoolBlend {blendOption=Linear}
    alphaScaleBiasClamp {bInterpResult=true, interpSpeedDecreasing=20}
    boneToModify {boneName=dyn_hair_bk_r_MSTR}
    rotation {pitch=30}
   }
   n7 = "Transform (Modify) Bone - Bone: dyn_hair_r_MSTR"(
    ComponentPose=@n6
    Alpha=Get is ParachuteOpened alpha() -> (is ParachuteOpened alpha:double)
    Translation="0,0,0"
    Rotation="0,0,30"
    Scale="1,1,1"
   ) {
    rotationMode=BMM_Additive
    rotationSpace=BCS_BoneSpace
    translationSpace=BCS_BoneSpace
    alphaBoolBlend {blendOption=Linear}
    alphaScaleBiasClamp {interpSpeedIncreasing=0.1}
    boneToModify {boneName=dyn_hair_r_MSTR}
    rotation {roll=30}
   }
   n8 = Get Override Ex Force() -> (Override Ex Force:Vector)
   n9 = Get gravity_override() -> (gravity_override:Vector)
   n10 = RigidBody!(ComponentPose=@n7, Alpha=Get Own Backpack() -> (Own Backpack:double), OverrideWorldGravity:Vector=@n9, ExternalForce=@n8) -> (Pose) {
    bClampLinearTranslationLimitToRefPose=true
    bOverrideWorldGravity=true
    overridePhysicsAsset=&/F_MED_Raider_Pink_FaceAcc_Hair_Dyn
    simulationSpace=BaseBoneSpace
    alphaBoolBlend {blendOption=Linear}
    baseBoneRef {boneName=neck_02}
    componentAppliedLinearAccClamp=<500 500 500>
    componentLinearAccScale=<0.3 0.3 0.3>
    componentLinearVelScale=<0.2 0.2 0.2>
    overrideWorldGravity {z=-680}
    simSpaceSettings {maxAngularAcceleration=5, maxAngularVelocity=0.1, maxLinearAcceleration=500, maxLinearVelocity=0.1, velocityScaleZ=0.1}
   }
   n11 = RigidBody(ComponentPose=@n10, Alpha=Get another backpack() -> (another backpack:double), OverrideWorldGravity=@n9, ExternalForce=@n8) {
    bClampLinearTranslationLimitToRefPose=true
    bOverrideWorldGravity=true
    overridePhysicsAsset=&F_MED_Raider_Pink_FaceAcc_Hair_Dyn
    simulationSpace=BaseBoneSpace
    alphaBoolBlend {blendOption=Linear}
    baseBoneRef {boneName=neck_02}
    componentAppliedLinearAccClamp=<500 500 200>
    componentLinearAccScale=<0.1 0.1 0.1>
    componentLinearVelScale=<0.2 0.2 0.2>
    overrideWorldGravity {z=-680}
    simSpaceSettings {maxAngularAcceleration=5, maxAngularVelocity=0.1, maxLinearAcceleration=500, maxLinearVelocity=0.1, velocityScaleZ=0.1}
   }
   n12 = Get is targeting() -> (is targeting:double)
   n13 = "Constraint - dyn_hair_r_01 (PP)"!(ComponentPose=@n11, Alpha=@n12, ConstraintWeights_0=0.2, ConstraintWeights_1:float=0.8) -> (Pose) {
    alphaBoolBlend {blendOption=Linear}
    alphaScaleBiasClamp {bInterpResult=true, interpSpeedDecreasing=3}
    boneToModify {boneName=dyn_hair_r_01}
    constraintSetup [{targetBone {boneName=neck_01}}, {targetBone {boneName=head}}]
    constraintWeights [1,1]
   }
   n14 = "Constraint - dyn_hair_r_02 (PP)"!(ComponentPose=@n13, Alpha=@n12, ConstraintWeights_0=0.5, ConstraintWeights_1:float=0.5) -> (Pose) {
    alphaBoolBlend {blendOption=Linear}
    alphaScaleBiasClamp {bInterpResult=true, interpSpeedDecreasing=2}
    boneToModify {boneName=dyn_hair_r_02}
    constraintSetup [{targetBone {boneName=neck_01}}, {targetBone {boneName=head}}]
    constraintWeights [1,1]
   }
   n15 = "Constraint - dyn_hair_r_03 (PP)"!(ComponentPose=@n14, Alpha=@n12, ConstraintWeights_0=0.5, ConstraintWeights_1:float=0.5) -> (Pose) {
    alphaBoolBlend {blendOption=Linear}
    alphaScaleBiasClamp {bInterpResult=true, interpSpeedDecreasing=1.5}
    boneToModify {boneName=dyn_hair_r_03}
    constraintSetup [{targetBone {boneName=spine_05}}, {targetBone {boneName=head}}]
    constraintWeights [1,1]
   }
   n16 = "Constraint - dyn_hair_r_04 (PP)"!(ComponentPose=@n15, Alpha=@n12, ConstraintWeights_0=0.75, ConstraintWeights_1:float=0.25) -> (Pose) {
    alphaBoolBlend {blendOption=Linear}
    alphaScaleBiasClamp {bInterpResult=true, interpSpeedDecreasing=1}
    boneToModify {boneName=dyn_hair_r_04}
    constraintSetup [{targetBone {boneName=spine_05}}, {targetBone {boneName=head}}]
    constraintWeights [1,1]
   }
   n17 = "Constraint - dyn_hair_r_05 (P)"(ComponentPose=@n16, Alpha=@n12, ConstraintWeights_0=1) {
    alphaBoolBlend {blendOption=Linear}
    alphaScaleBiasClamp {bInterpResult=true, interpSpeedDecreasing=0.5}
    boneToModify {boneName=dyn_hair_r_05}
    constraintSetup [{targetBone {boneName=spine_05}}]
    constraintWeights [1]
   }
   n18 = "Constraint - dyn_hair_r_06 (P)"(ComponentPose=@n17, Alpha=@n12, ConstraintWeights_0=1) {
    alphaBoolBlend {blendOption=Linear}
    alphaScaleBiasClamp {bInterpResult=true, interpSpeedDecreasing=0.3}
    boneToModify {boneName=dyn_hair_r_06}
    constraintSetup [{targetBone {boneName=spine_05}}]
    constraintWeights [1]
   }
   n19 = Component To Local(ComponentPose=@n18)
   n20 = "Constraint - dyn_hair_r_01 (PP)"(ComponentPose=Local To Component(LocalPose=Use cached pose 'preHairPose'()), Alpha=1, ConstraintWeights_0=0.2, ConstraintWeights_1=0.8) {
    alphaBoolBlend {blendOption=Linear}
    boneToModify {boneName=dyn_hair_r_01}
    constraintSetup [{targetBone {boneName=neck_01}}, {targetBone {boneName=head}}]
    constraintWeights [1,1]
   }
   n21 = "Constraint - dyn_hair_r_02 (PP)"(ComponentPose=@n20, Alpha=1, ConstraintWeights_0=0.5, ConstraintWeights_1=0.5) {
    alphaBoolBlend {blendOption=Linear}
    boneToModify {boneName=dyn_hair_r_02}
    constraintSetup [{targetBone {boneName=neck_01}}, {targetBone {boneName=head}}]
    constraintWeights [1,1]
   }
   n22 = "Constraint - dyn_hair_r_03 (PP)"(ComponentPose=@n21, Alpha=1, ConstraintWeights_0=0.5, ConstraintWeights_1=0.5) {
    alphaBoolBlend {blendOption=Linear}
    boneToModify {boneName=dyn_hair_r_03}
    constraintSetup [{targetBone {boneName=spine_05}}, {targetBone {boneName=head}}]
    constraintWeights [1,1]
   }
   n23 = "Constraint - dyn_hair_r_04 (PP)"(ComponentPose=@n22, Alpha=1, ConstraintWeights_0=0.75, ConstraintWeights_1=0.25) {
    alphaBoolBlend {blendOption=Linear}
    boneToModify {boneName=dyn_hair_r_04}
    constraintSetup [{targetBone {boneName=spine_05}}, {targetBone {boneName=head}}]
    constraintWeights [1,1]
   }
   n24 = "Constraint - dyn_hair_r_05 (P)"(ComponentPose=@n23, Alpha=1, ConstraintWeights_0=1) {
    alphaBoolBlend {blendOption=Linear}
    boneToModify {boneName=dyn_hair_r_05}
    constraintSetup [{targetBone {boneName=spine_05}}]
    constraintWeights [1]
   }
   n25 = "Constraint - dyn_hair_r_06 (P)"(ComponentPose=@n24, Alpha=1, ConstraintWeights_0=1) {
    alphaBoolBlend {blendOption=Linear}
    boneToModify {boneName=dyn_hair_r_06}
    constraintSetup [{targetBone {boneName=spine_05}}]
    constraintWeights [1]
   }
   n26 = Look At - Bone: dyn_hair_bk_01(ComponentPose=@n25, Alpha=1) {
    alphaBoolBlend {blendOption=Linear}
    boneToModify {boneName=dyn_hair_bk_01}
    lookAtLocation=<0 -10>
    lookAtTarget {boneReference {boneName=spine_05}}
    lookAt_Axis {axis=<1 0>}
    lookUp_Axis {axis {x=0}}
   }
   n27 = Look At - Bone: dyn_hair_bk_l_01(ComponentPose=@n26, Alpha=1) {
    alphaBoolBlend {blendOption=Linear}
    boneToModify {boneName=dyn_hair_bk_l_01}
    lookAtLocation=<0 -3 -11>
    lookAtTarget {boneReference {boneName=spine_05}}
    lookAt_Axis {axis=<1 0>}
    lookUp_Axis {axis {x=0}}
   }
   n28 = Look At - Bone: dyn_hair_bk_r_01(ComponentPose=@n27, Alpha=1) {
    alphaBoolBlend {blendOption=Linear}
    boneToModify {boneName=dyn_hair_bk_r_01}
    lookAtLocation=<0 -8 5>
    lookAtTarget {boneReference {boneName=spine_05}}
    lookAt_Axis {axis=<1 0>}
    lookUp_Axis {axis {x=0}}
   }
   n29 = "Constraint - dyn_hair_bk_l_02 (PP)"!(ComponentPose=@n28, Alpha=1, ConstraintWeights_0=0.5, ConstraintWeights_1:float=0.5) -> (Pose) {
    alphaBoolBlend {blendOption=Linear}
    boneToModify {boneName=dyn_hair_bk_l_02}
    constraintSetup [{targetBone {boneName=spine_05}}, {targetBone {boneName=head}}]
    constraintWeights [1,1]
   }
   n30 = "Constraint - dyn_hair_bk_l_03 (PP)"!(ComponentPose=@n29, Alpha=1, ConstraintWeights_0=0.8, ConstraintWeights_1:float=0.2) -> (Pose) {
    alphaBoolBlend {blendOption=Linear}
    boneToModify {boneName=dyn_hair_bk_l_03}
    constraintSetup [{targetBone {boneName=spine_05}}, {targetBone {boneName=head}}]
    constraintWeights [1,1]
   }
   n31 = "Constraint - dyn_hair_bk_l_04 (P)"(ComponentPose=@n30, Alpha=1, ConstraintWeights_0=1) {
    alphaBoolBlend {blendOption=Linear}
    boneToModify {boneName=dyn_hair_bk_l_04}
    constraintSetup [{targetBone {boneName=spine_05}}]
    constraintWeights [1]
   }
   n32 = "Constraint - dyn_hair_bk_02 (PP)"!(ComponentPose=@n31, Alpha=1, ConstraintWeights_0=0.5, ConstraintWeights_1:float=0.5) -> (Pose) {
    alphaBoolBlend {blendOption=Linear}
    boneToModify {boneName=dyn_hair_bk_02}
    constraintSetup [{targetBone {boneName=spine_05}}, {targetBone {boneName=head}}]
    constraintWeights [1,1]
   }
   n33 = "Constraint - dyn_hair_bk_03 (PP)"!(ComponentPose=@n32, Alpha=1, ConstraintWeights_0=0.8, ConstraintWeights_1:float=0.2) -> (Pose) {
    alphaBoolBlend {blendOption=Linear}
    boneToModify {boneName=dyn_hair_bk_03}
    constraintSetup [{targetBone {boneName=spine_05}}, {targetBone {boneName=head}}]
    constraintWeights [1,1]
   }
   n34 = "Constraint - dyn_hair_bk_04 (P)"(ComponentPose=@n33, Alpha=1, ConstraintWeights_0=1) {
    alphaBoolBlend {blendOption=Linear}
    boneToModify {boneName=dyn_hair_bk_04}
    constraintSetup [{targetBone {boneName=spine_05}}]
    constraintWeights [1]
   }
   n35 = "Constraint - dyn_hair_bk_r_02 (PP)"!(ComponentPose=@n34, Alpha=1, ConstraintWeights_0=0.5, ConstraintWeights_1:float=0.5) -> (Pose) {
    alphaBoolBlend {blendOption=Linear}
    boneToModify {boneName=dyn_hair_bk_r_02}
    constraintSetup [{targetBone {boneName=spine_05}}, {targetBone {boneName=head}}]
    constraintWeights [1,1]
   }
   n36 = "Constraint - dyn_hair_bk_r_03 (PP)"!(ComponentPose=@n35, Alpha=1, ConstraintWeights_0=0.8, ConstraintWeights_1:float=0.2) -> (Pose) {
    alphaBoolBlend {blendOption=Linear}
    boneToModify {boneName=dyn_hair_bk_r_03}
    constraintSetup [{targetBone {boneName=spine_05}}, {targetBone {boneName=head}}]
    constraintWeights [1,1]
   }
   n37 = "Constraint - dyn_hair_bk_r_04 (P)"(ComponentPose=@n36, Alpha=1, ConstraintWeights_0=1) {
    alphaBoolBlend {blendOption=Linear}
    boneToModify {boneName=dyn_hair_bk_r_04}
    constraintSetup [{targetBone {boneName=spine_05}}]
    constraintWeights [1]
   }
   n38 = "Constraint - dyn_hair_l_02 (PP)"!(ComponentPose=@n37, Alpha=1, ConstraintWeights_0=0.2, ConstraintWeights_1:float=0.8) -> (Pose) {
    alphaBoolBlend {blendOption=Linear}
    boneToModify {boneName=dyn_hair_l_02}
    constraintSetup [{targetBone {boneName=spine_05}}, {targetBone {boneName=head}}]
    constraintWeights [1,1]
   }
   n39 = "Constraint - dyn_hair_l_03 (PP)"!(ComponentPose=@n38, Alpha=1, ConstraintWeights_0=0.5, ConstraintWeights_1:float=0.5) -> (Pose) {
    alphaBoolBlend {blendOption=Linear}
    boneToModify {boneName=dyn_hair_l_03}
    constraintSetup [{targetBone {boneName=spine_05}}, {targetBone {boneName=head}}]
    constraintWeights [1,1]
   }
   n40 = "Constraint - dyn_hair_l_04 (PP)"!(ComponentPose=@n39, Alpha=1, ConstraintWeights_0=0.8, ConstraintWeights_1:float=0.2) -> (Pose) {
    alphaBoolBlend {blendOption=Linear}
    boneToModify {boneName=dyn_hair_l_04}
    constraintSetup [{targetBone {boneName=spine_05}}, {targetBone {boneName=head}}]
    constraintWeights [1,1]
   }
   n41 = Component To Local(ComponentPose=@n40)
   n42 = Get isRigidBodyDisable() -> (isRigidBodyDisable:bool)
   n43 = "Transform (Modify) Bone - Bone: C_jaw"(
    ComponentPose=Local To Component(LocalPose=Blend Poses by bool(BlendPose_0=@n41, BlendPose_1=@n19, bActiveValue=@n42) {blendType=Linear})
    Alpha=Get lod 1 emote jaw correction() -> (lod 1 emote jaw correction:double)
    Translation="0,0.7,-0.1"
    Rotation="0,0,0"
    Scale="1,1,1"
   ) {
    translationMode=BMM_Additive
    translationSpace=BCS_BoneSpace
    alphaBoolBlend {blendOption=Linear}
    alphaScaleBiasClamp {interpSpeedDecreasing=2, interpSpeedIncreasing=5}
    boneToModify {boneName=C_jaw}
    translation {y=0.7, z=-0.1}
   }
   n44 = "Transform (Modify) Bone - Bone: dyn_glasses"(
    ComponentPose=@n43
    Alpha=Get lod 1 emote glasses correction() -> (lod 1 emote glasses correction:double)
    Translation="0,0.2,0"
    Rotation="0,0,-3"
    Scale="1,1,1"
   ) {
    rotationMode=BMM_Additive
    rotationSpace=BCS_BoneSpace
    translationMode=BMM_Additive
    translationSpace=BCS_BoneSpace
    alphaBoolBlend {blendOption=Linear}
    alphaScaleBiasClamp {interpSpeedDecreasing=2, interpSpeedIncreasing=5}
    boneToModify {boneName=dyn_glasses}
    rotation {roll=-3}
    translation {y=0.2}
   }
   n45 = Get is lod 1() -> (is lod 1:double)
   n46 = "Transform (Modify) Bone - Bone: C_jaw"(ComponentPose=@n44, Alpha=@n45, Translation="0,0.2,0", Rotation="0,0,0", Scale="1,1,1") {
    translationMode=BMM_Additive
    translationSpace=BCS_BoneSpace
    alphaBoolBlend {blendOption=Linear}
    alphaScaleBiasClamp {interpSpeedDecreasing=2, interpSpeedIncreasing=5}
    boneToModify {boneName=C_jaw}
    translation {y=0.2}
   }
   n47 = "Transform (Modify) Bone - Bone: dyn_glasses"(ComponentPose=@n46, Alpha=@n45, Translation="0,0.1,0", Rotation="0,0,0", Scale="1,1,1") {
    translationMode=BMM_Additive
    translationSpace=BCS_BoneSpace
    alphaBoolBlend {blendOption=Linear}
    alphaScaleBiasClamp {interpSpeedDecreasing=2, interpSpeedIncreasing=5}
    boneToModify {boneName=dyn_glasses}
    translation {y=0.1}
   }
   n48 = "Transform (Modify) Bone - Bone: neck_01_fr"(ComponentPose=@n47, Alpha=@n45, Translation="0,3,0", Rotation="0,0,0", Scale="1,1,1") {
    rotationMode=BMM_Additive
    rotationSpace=BCS_BoneSpace
    translationMode=BMM_Additive
    translationSpace=BCS_BoneSpace
    alphaBoolBlend {blendOption=Linear}
    boneToModify {boneName=neck_01_fr}
    translation {y=3}
   }
   n49 = "Transform (Modify) Bone - Bone: neck_01_bk"(ComponentPose=@n48, Alpha=@n45, Translation="0,-3,0", Rotation="0,0,0", Scale="1,1,1") {
    rotationMode=BMM_Additive
    rotationSpace=BCS_BoneSpace
    translationMode=BMM_Additive
    translationSpace=BCS_BoneSpace
    alphaBoolBlend {blendOption=Linear}
    boneToModify {boneName=neck_01_bk}
    translation {y=-3}
   }
   n50 = "Transform (Modify) Bone - Bone: neck_01_bk"(
    ComponentPose=@n49
    Alpha=Get LOD 0 Mask Correction() -> (LOD 0 Mask Correction:float)
    Translation="0,-1.5,0"
    Rotation="0,0,0"
    Scale="1,1,1"
   ) {
    rotationMode=BMM_Additive
    rotationSpace=BCS_BoneSpace
    translationMode=BMM_Additive
    translationSpace=BCS_BoneSpace
    alphaBoolBlend {blendOption=Linear}
    boneToModify {boneName=neck_01_bk}
    translation {y=-1.5}
   }
   n51 = "Transform (Modify) Bone - Bone: neck_01_lt"(ComponentPose=@n50, Alpha=@n45, Translation="0,0,-3", Rotation="0,0,0", Scale="1,1,1") {
    rotationMode=BMM_Additive
    rotationSpace=BCS_BoneSpace
    translationMode=BMM_Additive
    translationSpace=BCS_BoneSpace
    alphaBoolBlend {blendOption=Linear}
    boneToModify {boneName=neck_01_lt}
    translation {z=-3}
   }
   n52 = "Transform (Modify) Bone - Bone: neck_01_rt"(ComponentPose=@n51, Alpha=@n45, Translation="0,0,3", Rotation="0,0,0", Scale="1,1,1") {
    rotationMode=BMM_Additive
    rotationSpace=BCS_BoneSpace
    translationMode=BMM_Additive
    translationSpace=BCS_BoneSpace
    alphaBoolBlend {blendOption=Linear}
    boneToModify {boneName=neck_01_rt}
    translation {z=3}
   }
   n53 = Control Rig(Source=Component To Local(ComponentPose=@n52), Alpha=Get isLOD0() -> (isLOD0:double)) {
    alphaBoolBlend {blendOption=Linear}
    controlRigAssetReference {blueprintRigClass=&/F_MED_Raider_Pink_FaceAcc_CtrlRig.F_MED_Raider_Pink_FaceAcc_CtrlRig_C}
   }
   Output Pose(Result=@n53) {layerGroup=None}
   n54 = Copy Pose From Mesh(SourceMeshComponent=Get Body_Mesh())
   n55 = Copy Pose From Mesh() {bCopyCurves=true, bUseAttachedParent=true}
   n56 = Get head_rot_bk() -> (head_rot_bk:double)
   n57 = "Transform (Modify) Bone - Bone: dyn_hair_bk_MSTR"(
    ComponentPose=Local To Component(LocalPose=Layered blend per bone(BasePose=@n55, BlendPoses_0=@n54) {layerSetup [{branchFilters [{boneName=upperarm_l}, {boneName=upperarm_r}]}]})
    Alpha=@n56
    Translation="-2.162,-0.000,0.800"
    Rotation="50.000,0.000,0.000"
    Scale="1,1,1"
   ) {
    rotationMode=BMM_Additive
    rotationSpace=BCS_BoneSpace
    translationMode=BMM_Additive
    translationSpace=BCS_BoneSpace
    alphaBoolBlend {blendOption=Linear}
    boneToModify {boneName=dyn_hair_bk_MSTR}
    rotation {pitch=49.9999}
    translation {x=-2.162, z=0.8}
   }
   n58 = "Transform (Modify) Bone - Bone: dyn_hair_bk_l_MSTR"(ComponentPose=@n57, Alpha=@n56, Translation="0,0,0", Rotation="40,-30,20", Scale="1,1,1") {
    rotationMode=BMM_Additive
    rotationSpace=BCS_BoneSpace
    translationMode=BMM_Additive
    translationSpace=BCS_BoneSpace
    alphaBoolBlend {blendOption=Linear}
    boneToModify {boneName=dyn_hair_bk_l_MSTR}
    rotation=<rot 40 -30 20>
   }
   n59 = "Transform (Modify) Bone - Bone: dyn_hair_bk_r_MSTR"(ComponentPose=@n58, Alpha=@n56, Translation="0,0,0", Rotation="50,0,0", Scale="1,1,1") {
    rotationMode=BMM_Additive
    rotationSpace=BCS_BoneSpace
    translationMode=BMM_Additive
    translationSpace=BCS_BoneSpace
    alphaBoolBlend {blendOption=Linear}
    boneToModify {boneName=dyn_hair_bk_r_MSTR}
    rotation {pitch=50}
   }
   Save cached pose 'preHairPose'(Pose=Component To Local(ComponentPose=@n59)) {cacheName=preHairPose}
  ]}
 EventGraph {nodes [
   Event BlueprintInitializeAnimation()
   > Set Body_Mesh(Body_Mesh=GetOwnerSkeletalMeshComponent())
   Event OnParachuteOpened()
   > Delay(Duration=1.5)
   n1 = Event OnBeginSkydiving()
   Delay(execute=[@n1.then, @n2.then], Duration=0.125)
   > ResetDynamics(InTeleportType=TeleportPhysics)
   > ResetDynamics(InTeleportType=ResetPhysics)
   n2 = Event OnEndSkydiving()
   Event BlueprintUpdateAnimation()
   > n3 = Sequence!(execute) -> (then_0, then_1, then_2:Exec, then_3:Exec, then_4:Exec, then_5:Exec, then_6:Exec)
   n4 = BreakTransform(InTransform=GetDeltaTransformFromRefPose(self=GetOwnerSkeletalMeshComponent(), BoneName=head, BaseName=spine_05))
   > Set head_rot_bk(head_rot_bk:double=MapRangeClamped(Value=BreakRotator(InRot=@n4.Rotation).Yaw, InRangeA=0, InRangeB=-50)) -> (Output_Get:double)
   > n5 = Branch(Condition=Get bIsSwimSprinting() -> (bIsSwimSprinting:bool))
   > Set gravity_override(gravity_override="0, 0, 0")
   n6 = Branch(execute=@n5.else, Condition=Get bIsSkyDiving())
   > n7 = Branch(Condition=Get bIsParachuteOpened())
   Set gravity_override(execute=[@n6.else, @n7.then], gravity_override="0, 0,-780")
   Set gravity_override(execute=@n7.else, gravity_override="0, 0,-300")
   n8 = Branch(execute=@n3.then_1, Condition=Get endCrouchinn() -> (endCrouchinn:bool))
   > Set Override Ex Force(Override Ex Force:Vector="0,-5000,10000") -> (Output_Get:Vector)
   Set Override Ex Force(execute=@n8.else, Override Ex Force="0, 0, 0")
   n9 = Get CurrentLOD()
   n10 = Get bRigidBodyEnabled()
   Set isRigidBodyDisable(
    execute=@n3.then_2
    isRigidBodyDisable:bool=OR Boolean(A="integer > integer"(A=@n9, B=1) {is=title K2Node_CallFunction}, B=NOT Boolean(A=@n10))
   ) -> (Output_Get:bool)
   > n11 = Branch(Condition="Equal (Integer)"(A=Get CurrentLOD(), B=0) {is=title K2Node_CallFunction})
   > Set isLOD0(isLOD0:double=1) -> (Output_Get:double)
   Set isLOD0(execute=@n11.else, isLOD0=0)
   n12 = Branch(execute=@n3.then_5, Condition=Get bIsTargeting())
   > Gate!(Enter:Exec, Open:Exec=@n12.else, Close:Exec=@n13.then, Toggle:Exec, bStartClosed:bool=false) -> (Exit:Exec) {is=title K2Node_MacroInstance}
   > Set is targeting(is targeting:double=1) -> (Output_Get:double)
   > Delay(Duration=0.2)
   > n13 = Set is targeting(is targeting=0)
   n14 = Branch(execute=@n3.then_6, Condition="Equal (Integer)"(A=Get CurrentLOD(), B=1))
   > Set is lod 1(is lod 1:double=1) -> (Output_Get:double)
   Set is lod 1(execute=@n14.else, is lod 1=0)
   Set Is DBNO Float(execute=@n3.then_3, Is DBNO Float:double="To Float (Boolean)"(InBool=Get bIsDBNO())) -> (Output_Get:double)
   > Set is ParachuteOpened alpha(is ParachuteOpened alpha:double=SelectFloat(A=1, B=0, bPickA=Get bIsParachuteOpened())) -> (Output_Get:double)
   > Set isWearingBackpack(isWearingBackpack:bool=CheckForBackbling()) -> (Output_Get:bool)
   n15 = Get isWearingBackpack() -> (isWearingBackpack:bool)
   > n16 = Branch(Condition=OR Boolean(A=Get is Own Backpack() -> (is Own Backpack:bool), B=NOT Boolean(A=@n15)))
   > Set Own Backpack(Own Backpack:double=1) -> (Output_Get:double)
   > Set another backpack(another backpack:double=0) -> (Output_Get:double)
   Set Own Backpack(execute=@n16.else, Own Backpack=0)
   > Set another backpack(another backpack=1)
   n17 = Get Body_Mesh()
   n18 = AND Boolean!(A=Get bIsCrouching(), B=Get bIsMoving2D(), C:bool=Get bIsTargeting(), D:bool=NOT Boolean(A=Get Going Comando() -> (Going Comando:bool))) -> (ReturnValue)
   "Set Head Pitch When Crouching "(
    execute=@n3.then_4
    "Head Pitch When Crouching ":double=SelectFloat(
     A=GetAlphaFromReferenceTransformDelta(SkeletalMeshComponent=@n17, BoneName=head, BaseName=spine_05, InputRange=(X=-40,Y=-80), TwistAxis="1,0,0")
     B=0
     bPickA=@n18
    )
   ) -> (Output_Get:double)
   > "Set Head Roll When Crouching "("Head Roll When Crouching ":float=SelectFloat(
     A=GetAlphaFromReferenceTransformDelta(SkeletalMeshComponent=@n17, BoneName=head, BaseName=spine_05, DeltaComponentType=EulerY, InputRange=(X=20,Y=40), TwistAxis="1,0,0")
     B=0
     bPickA=@n18
    )) -> (Output_Get:float)
   n19 = Get CurrentLOD()
   > Set LOD 0 Mask Correction(LOD 0 Mask Correction:float=SelectFloat(
     A=GetAlphaFromReferenceTransformDelta(SkeletalMeshComponent=@n17, BoneName=head, BaseName=spine_05, InputRange=(X=0,Y=-40), TwistAxis="1,0,0")
     B=0
     bPickA="Equal (Integer)"(A=@n19, B=0)
    )) -> (Output_Get:float)
   n20 = Get CurrentLOD()
   n21 = SelectFloat(A=1, B=0, bPickA=AND Boolean!(A=Get bIsPlayingEmote(), B="Equal (Integer)"(A=@n20, B=1)) -> (ReturnValue))
   > Set lod 1 emote jaw correction(lod 1 emote jaw correction:double=@n21) -> (Output_Get:double)
   > Set lod 1 emote glasses correction(lod 1 emote glasses correction:double=@n21) -> (Output_Get:double)
   > n22 = Cast To FortPlayerAnimInstance(Object=GetAnimInstance(self=GetOwnerSkeletalMeshComponent()))
   > Set Going Comando(Going Comando:bool=false) -> (Output_Get:bool)
   > n23 = Switch on EFortWeaponCoreAnimation!(execute, Selection:EFortWeaponCoreAnimation Enum=Get WeaponCoreAnim(self=@n22)) -> (
    Melee:Exec
    Pistol:Exec
    Shotgun:Exec
    PaperBlueprint:Exec
    Rifle:Exec
    MeleeOneHand:Exec
    MachinePistol:Exec
    RocketLauncher:Exec
    GrenadeLauncher:Exec
    GoingCommando:Exec
    AssaultRifle:Exec
    TacticalShotgun:Exec
    SniperRifle:Exec
    TrapPlacement:Exec
    ShoulderLauncher:Exec
    AbilityDecoTool:Exec
    Crossbow:Exec
    C4:Exec
    RemoteControl:Exec
    DualWield:Exec
    AR_BullPup:Exec
    AR_ForwardGrip:Exec
    MedPackPaddles:Exec
    SMG_P90:Exec
    AR_DrumGun:Exec
    Consumable_Small:Exec
    Consumable_Large:Exec
    Balloon:Exec
    MountedTurret:Exec
    CreativeTool:Exec
    ExplosiveBow:Exec
    AshtonIndigo:Exec
    AshtonChicago:Exec
    MeleeDualWield:Exec
    Unarmed:Exec
   )
   Set Going Comando(execute=@n23.GoingCommando, Going Comando=true)
   n24 = Event OnCharacterCustomizationLoaded()
   > Set BackPack Equipped(BackPack Equipped:Skeletal Mesh Component Budgeted Object Reference=Get Array Item with EFort Custom Part Type(Array=@n24.CustomMeshParts, PartType=Backpack)) -> (Output_Get:Skeletal Mesh Component Budgeted Object Reference)
   > n25 = Get!(execute:Exec) -> (then:Exec, else:Exec, BackPack Equipped:Skeletal Mesh Component Budgeted Object Reference)
   > Set is Own Backpack(is Own Backpack:bool=Contains(SearchIn=GetDisplayName(Object=@n25), Substring=Raider_Pink)) -> (Output_Get:bool)
   Event OnEndCrouching()
   > Set endCrouchinn(endCrouchinn:bool=true) -> (Output_Get:bool)
   > Delay(Duration=0.25)
   > Set endCrouchinn(endCrouchinn=false)
   Event OnBeginCrouching()
  ]}
]
variables [
 another backpack {type=number}
 BackPack Equipped {type=SkeletalMeshComponentBudgeted}
 Body_Mesh {type=SkeletalMeshComponent}
 endCrouchinn {type=boolean}
 Going Comando {type=boolean}
 gravity_override {type=Vector, default=<0 0 -780>}
 Hair corr not mov {type=number}
 "Head Pitch When Crouching " {type=number}
 "Head Roll When Crouching " {type=number}
 head_rot_bk {type=number}
 Is DBNO Float {type=number}
 is lod 1 {type=number}
 is Own Backpack {type=boolean}
 is ParachuteOpened alpha {type=number}
 is targeting {type=number}
 isLOD0 {type=number}
 isRigidBodyDisable {type=boolean}
 isWearingBackpack {type=boolean}
 LOD 0 Mask Correction {type=number}
 lod 1 emote glasses correction {type=number}
 lod 1 emote jaw correction {type=number}
 Override Ex Force {type=Vector}
 Own Backpack {type=number}
]
